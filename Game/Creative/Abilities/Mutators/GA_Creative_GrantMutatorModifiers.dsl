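; Creative mutator gameplay ability that grants a move speed modifier to its owner while active.
; On activation, a ShouldApplyModifier check supplies the magnitude for the speed effect, set through SetByCaller.MoveSpeed, and the applied effect handle is stored.
; Ending the ability removes all stored effects and clears the list.
; GetSpeedMultiplier returns the scalable-float setting when non-negative, otherwise the move speed knob, plus a flag saying which applied.
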(event Ability|EventActivateAbility
  (bind _ge_speedmodifier (Variables|GameplayEffects|GetGE_SpeedModifier))
  (|CallParentFunction)
  (Utilities|FlowControl|Sequence
    (:then_0
      (bind _magnitude (Utilities|ShouldApplyModifier _ge_speedmodifier (Variables|Default|GetSpeedModifierName))
        (:Yes
          (bind _returnvalue (Ability|GameplayEffect|AssignTagSetbyCallerMagnitude (Ability|MakeOutgoingGameplayEffectSpec _ge_speedmodifier) "(TagName=\"SetByCaller.MoveSpeed\")" _magnitude))
          (bind _returnvalue_2 (Ability|ApplyGameplayEffectSpecToOwner _returnvalue))
          (Utilities|Array|Add (Variables|GameplayEffects|GetActiveEffectSpecs) _returnvalue_2))))))

(event Ability|EventFailedToActivatePassiveAbility
  (params (ActorInfo :type Gameplay%20Ability%20Actor%20Info%20Structure)))

(event Ability|EventOnEndAbility
  (params (bWasCancelled :type Boolean))
  (|CallParentFunction bWasCancelled)
  (for _array_element (Variables|GameplayEffects|GetActiveEffectSpecs)
    (Ability|RemoveGameplayEffectFromOwnerWithHandle _array_element))
  (Utilities|Array|Clear (Variables|GameplayEffects|GetActiveEffectSpecs)))

(fn GetSpeedMultiplier
  (bind _speedmodifiersetting (Variables|Default|GetSpeedModifierSetting))
  (Utilities|FlowControl|Sequence
    (:then_0
      (if (ScalableFloat|IsSettingCurveValid _speedmodifiersetting)
        (LocalVariables|SetSpeedModifierValue (ScalableFloat|GetValueatLevelwithContext _speedmodifiersetting))))
    (:then_1
      (return (select (>= (LocalVariables|GetSpeedModifierValue) 0) (LocalVariables|GetSpeedModifierValue) (Variables|Knobs|GetKnob_fMoveSpeedMultiplier)) (>= (LocalVariables|GetSpeedModifierValue) 0)))))
